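# Translucent, two-sided glass material for the RL131 cosmetic wheel, with all of its look built in emissive.
# A clamped parallax-corrected environment map is weighted by a squared, inverted normal-map-to-camera dot product, then combined with a Fresnel rim, both masked by vertex color.
# A tint is added, and OverlayENV_Intensity blends in extra environment reflection.
# The result passes through the shared vehicle damage and de-rez transition functions.
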Material /VehicleCosmetics/Wheels/RL131/Materials/M_Wheel_RL131_Glass
graph {
 properties {
  bAutomaticallySetUsageInEditor=false
  bUseMaterialAttributes=true
  bUsedWithSkeletalMesh=true
  blendMode=BLEND_Translucent
  twoSided=true
  usedShadingModels=Default Lit
 }
 nodes [
  n1 = Constant()
  n2 = Clamp(
   None=MaterialFunctionCall!(Desaturate, Tint=@n1, TintAmount=@n1, ParallaxSize) {materialFunction=/VehicleCosmetics/Wheels/Mild/Materials/MF/MF_ParallexCorrectWorldENV}.WorldSpace
   Min=ScalarParameter() {parameterName=ENVMap_ClampMin}
   Max=Constant() {r=3}
  )
  n3 = TextureSampleParameter2D() {parameterName=Normal, samplerType=SAMPLERTYPE_Normal, texture=/VehicleCosmetics/Bodies/Ramen/Skin/Ramen_Dragonlord/Blank_N}
  n4 = CameraVectorWS()
  n5 = OneMinus(None=Clamp(None=DotProduct(A=@n3.RGB, B=@n4)))
  n6 = Multiply(A=@n2, B=Add(A=Multiply(A=@n5, B=@n5)) {constB=0.25})
  n7 = VertexColor()
  n8 = Fresnel() {exponent=4}
  n9 = Add(A=VectorParameter() {parameterName=Tint}.RGB, B=Add(A=Multiply(A=@n6, B=@n7), B=Multiply(A=@n7, B=@n8)))
  n10 = LinearInterpolate(A=@n9, B=Add(A=Multiply(A=@n2) {constB=2}, B=@n9), Alpha=Clamp(None=ScalarParameter() {parameterName=OverlayENV_Intensity}))
  MaterialFunctionCall(In=MaterialFunctionCall(Material=SetMaterialAttributes(Emissive Color=@n10) {attributeSetTypes [#1]}) {materialFunction=&MF_BaseVehicle_Damage}.Result) {materialFunction=&MF_VehicleDeRez_Transition}
 ]
}
