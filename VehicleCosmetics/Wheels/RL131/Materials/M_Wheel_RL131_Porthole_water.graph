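# Additive two-sided vehicle material for the porthole water on the RL131 wheel, built on material attributes.
# A parallax-corrected world environment reflection is weighted by a squared view-to-normal falloff and a named reroute, then a clamped double-Fresnel rim is added.
# The result drives emissive and base color, then passes through the vehicle damage and de-rez transition functions.
# A separate gradient chain using vertex color, UV and depth fade feeds no output.
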Material /VehicleCosmetics/Wheels/RL131/Materials/M_Wheel_RL131_Porthole_water
graph {
 properties {
  bAutomaticallySetUsageInEditor=false
  bUseMaterialAttributes=true
  bUsedWithSkeletalMesh=true
  blendMode=BLEND_Additive
  twoSided=true
  usedShadingModels=Default Lit
 }
 nodes [
  n1 = Constant()
  n2 = MaterialFunctionCall!(Desaturate, Tint=@n1, TintAmount=@n1, ParallaxSize) {materialFunction=/VehicleCosmetics/Wheels/Mild/Materials/MF/MF_ParallexCorrectWorldENV}
  n3 = TextureSampleParameter2D() {parameterName=Normal, samplerType=SAMPLERTYPE_Normal, texture=/VehicleCosmetics/Bodies/Ramen/Skin/Ramen_Dragonlord/Blank_N}
  n4 = CameraVectorWS()
  n5 = OneMinus(None=Clamp(None=DotProduct(A=@n3.RGB, B=@n4)))
  n6 = Add(A=Multiply(A=@n5, B=@n5)) {constB=0.025}
  n7 = Fresnel() {baseReflectFraction=0.02, exponent=2}
  n8 = Fresnel() {exponent=4}
  n9 = Add(A=Multiply(A=Multiply(A=@n2.WorldSpace, B=@n6), B=NamedRerouteUsage()), B=Clamp(None=Multiply(A=@n7, B=@n8)))
  MaterialFunctionCall(In=MaterialFunctionCall(Material=SetMaterialAttributes!(MaterialAttributes, Emissive Color=@n9, Base Color=@n9) {attributeSetTypes [#1, #2]}) {materialFunction=&MF_BaseVehicle_Damage}.Result) {materialFunction=&MF_VehicleDeRez_Transition}
  NamedRerouteDeclaration(None=VectorParameter() {parameterName=Water_Color, defaultValue=<rgba 0.123 0.727 0.967 1>}.RGB) {name=WaterColor, nodeColor=<rgba 0 0.9531 1 1>}
  n10 = NamedRerouteUsage()
  n11 = TextureCoordinate()
  n12 = Clamp(None=OneMinus(None=DepthFade() {fadeDistanceDefault=10, opacityDefault=10}))
  n13 = Add(A=Add(A=VertexColor().R, B=ComponentMask(None=@n11) {g=true}), B=Multiply(B=Power(Base=@n12) {constExponent=3}) {constA=0.65})
  Power(Base=Add(A=@n8, B=LinearInterpolate(B=@n10, Alpha=@n13))) {constExponent=2.5}
 ]
}
