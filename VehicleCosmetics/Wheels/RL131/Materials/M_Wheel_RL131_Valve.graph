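# Material for the valve on the RL131 vehicle wheel, animating it through World Position Offset.
# A sine of time plus scaled world Z moves vertices along the Y axis, with parameters for wave depth and a constant offset.
# That offset is set on the attributes from the simple wheel parent function, which then pass through the vehicle damage and de-rez transition functions.
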Material /VehicleCosmetics/Wheels/RL131/Materials/M_Wheel_RL131_Valve
graph {
 properties {
  bAutomaticallySetUsageInEditor=false
  bUseMaterialAttributes=true
  bUsedWithMeshDeformer=true
  bUsedWithSkeletalMesh=true
  pixelDepthOffsetMode=PDOM_Legacy
  usedShadingModels=Default Lit
 }
 nodes [
  n1 = MaterialFunctionCall() {materialFunction=&MF_WheelParent_Simple}
  n2 = WorldPosition()
  n3 = Sine(None=Add(A=Time(), B=Multiply(A=@n2.Z) {constB=0.035}))
  n4 = Constant()
  n5 = AppendVector(A=AppendVector(A=@n4, B=Constant() {r=2}), B=@n4)
  n6 = Add(
   A=Multiply(A=Multiply(A=@n3, B=@n5), B=ScalarParameter() {defaultValue=0.69, parameterName=wavedeph})
   B=Multiply(A=Constant3Vector() {constant=<rgba 0 1 0 0>}, B=ScalarParameter() {defaultValue=-0.7, parameterName=valve_offcet})
  )
  MaterialFunctionCall(In=MaterialFunctionCall(Material=SetMaterialAttributes!(MaterialAttributes=@n1.MaterialAttributes, World Position Offset=@n6) {attributeSetTypes [#1]}) {materialFunction=&MF_BaseVehicle_Damage}.Result) {materialFunction=&MF_VehicleDeRez_Transition}
 ]
}
