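# Additive, unlit, two-sided material for the water and bubbles in the RL131 wheel porthole.
# Emissive sums a water-tinted, parallax-corrected environment reflection, weighted by normal-map view angle, with a UV-gradient blend between Fresnel glow and two bubble channels panning at different speeds, masked by vertex red.
# The result passes through vehicle damage and de-rez transition functions.
# A separate depth-fade water-color branch ends unconnected.
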Material /VehicleCosmetics/Wheels/RL131/Materials/M_Wheel_RL131_porthole_water_bubs
graph {
 properties {
  bAutomaticallySetUsageInEditor=false
  bUseMaterialAttributes=true
  bUsedWithSkeletalMesh=true
  blendMode=BLEND_Additive
  shadingModel=MSM_Unlit
  twoSided=true
  usedShadingModels=Unlit
 }
 nodes [
  n1 = Constant()
  n2 = MaterialFunctionCall!(Desaturate, Tint=@n1, TintAmount=@n1, ParallaxSize) {materialFunction=/VehicleCosmetics/Wheels/Mild/Materials/MF/MF_ParallexCorrectWorldENV}
  n3 = TextureSampleParameter2D() {parameterName=Normal, samplerType=SAMPLERTYPE_Normal, texture=/VehicleCosmetics/Bodies/Ramen/Skin/Ramen_Dragonlord/Blank_N}
  n4 = CameraVectorWS()
  n5 = OneMinus(None=Clamp(None=DotProduct(A=@n3.RGB, B=@n4)))
  n6 = Add(A=Multiply(A=@n5, B=@n5)) {constB=0.025}
  n7 = Fresnel() {exponent=4}
  n8 = Multiply(A=Fresnel() {baseReflectFraction=0.02, exponent=2}, B=@n7)
  n9 = TextureCoordinate()
  n10 = Add(
   A=TextureSample(UVs=Panner(Coordinate=@n9) {speedY=0.5}) {samplerType=SAMPLERTYPE_LinearColor, texture=/VehicleCosmetics/Wheels/RL131/Textures/T_RL131_Bubbles_M}.R
   B=TextureSample(UVs=Panner(Coordinate=@n9) {speedY=0.75}) {samplerType=SAMPLERTYPE_LinearColor, texture=&T_RL131_Bubbles_M}.G
  )
  n11 = VertexColor()
  n12 = Multiply(A=Multiply(A=Multiply(A=@n10) {constB=10}, B=OneMinus(None=@n11.R))) {constB=0.2}
  n13 = Add(B=ComponentMask(None=TextureCoordinate()) {g=true}) {constA=-0.06}
  n14 = Add(
   A=Multiply(A=Multiply(A=@n2.WorldSpace, B=@n6), B=NamedRerouteUsage())
   B=LinearInterpolate(A=Clamp(None=@n8), B=Multiply(A=@n12, B=@n13), Alpha=@n13) {constB=0}
  )
  MaterialFunctionCall(In=MaterialFunctionCall(Material=SetMaterialAttributes(Emissive Color=@n14) {attributeSetTypes [#1]}) {materialFunction=&MF_BaseVehicle_Damage}.Result) {materialFunction=&MF_VehicleDeRez_Transition}
  NamedRerouteDeclaration(None=VectorParameter() {parameterName=Water_Color, defaultValue=<rgba 0.123 0.727 0.967 1>}.RGB) {name=WaterColor, nodeColor=<rgba 0 0.9531 1 1>}
  n15 = NamedRerouteUsage()
  n16 = TextureCoordinate()
  n17 = Clamp(None=OneMinus(None=DepthFade() {fadeDistanceDefault=10, opacityDefault=10}))
  n18 = Add(A=Add(A=VertexColor().R, B=ComponentMask(None=@n16) {g=true}), B=Multiply(B=Power(Base=@n17) {constExponent=3}) {constA=0.65})
  Power(Base=Add(A=@n7, B=LinearInterpolate(B=@n15, Alpha=@n18))) {constExponent=2.5}
 ]
}
